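# Physics asset for the Frost Mystery backpack cosmetic, defining simple collision bodies for the pack and its two wings.
# Each of the pack, left wing and right wing bones carries a single capsule, with the wing capsules larger and mirrored.
# Two constraints join each wing to the pack, sharing the same swing and twist angle limits on both sides.
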PhysicsAsset /BRCosmetics/Accessories/FORT_Backpacks/M_MED_Frost_Mystery_Backpack/Meshes/M_MED_Frost_Mystery_Pack_PhysicsAsset
bodies {
 pack [{shapeType=Capsule, length=18.0193, radius=4.7971, center=<0.0001 -0.8999 -0.4561>, rotation=<rot 0 -90.0008 -89.9977>}]
 wing_l [{shapeType=Capsule, length=25.1479, radius=7.847, center=<-0.6275 -0.9077 -16.6584>, rotation=<rot 0.49 -90 0.2314>}]
 wing_r [{shapeType=Capsule, length=25.1479, radius=7.847, center=<0.6275 0.9077 16.6584>, rotation=<rot 0.49 -90 0.2314>}]
}
constraints [
 {bone1Name=wing_l, bone2Name=pack, swing1LimitDegrees=45, swing2LimitDegrees=45, twistLimitDegrees=45}
 {bone1Name=wing_r, bone2Name=pack, swing1LimitDegrees=45, swing2LimitDegrees=45, twistLimitDegrees=45}
]
